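# Sound cue for the small-scale grind rail boost intro, playing footsteps layered with a small-rail sweetener.
# An IsLocal branch picks first-person attenuation when true, third-person when false or unset.
# Both paths mix the same two waves through a randomizing modulator, differing only in attenuation asset.
# The cue uses the gadgets sound class and carries occlusion and limit-to-owner concurrency override settings.
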SoundCue /GrindRail/Audio/SmallScaleGrinding/SmallScale_GrindRail_Boost_Intro_Steps_Seq_Cue
root=Branch_0
nodes [
 {id=Attenuation_0, children [Modulator_1], properties {attenuationSettings=/GrindRail/Audio/Grinding_1P_Att}}
 {id=Attenuation_1, children [Modulator_2], properties {attenuationSettings=/GrindRail/Audio/Grinding_3P_Att}}
 {id=Branch_0, children [Attenuation_0, Attenuation_1, Attenuation_1], properties {boolParameterName=IsLocal}}
 {id=Mixer_0, children [WavePlayer_1,WavePlayer_0], properties {inputVolume [0.6,0.6]}}
 {id=Mixer_1, children [WavePlayer_3,WavePlayer_2], properties {inputVolume [0.6,0.6]}}
 {id=Modulator_1, children [Mixer_0], properties {volumeMax=1.2}}
 {id=Modulator_2, children [Mixer_1], properties {volumeMax=1.2}}
 {id=WavePlayer_0, properties {soundWaveAssetPtr=/GrindRail/Audio/Grind_Rail_Boost_Intro_Steps_Only}}
 {id=WavePlayer_1, properties {soundWaveAssetPtr=&/Small_GrindRail_Sweetner}}
 {id=WavePlayer_2, properties {soundWaveAssetPtr=&Grind_Rail_Boost_Intro_Steps_Only}}
 {id=WavePlayer_3, properties {soundWaveAssetPtr=&Small_GrindRail_Sweetner}}
]
properties {
 maxDistance=6500
 soundClassObject=&Fort_Gadgets
 volumeMultiplier=1.511
 attenuationOverrides {
  bEnableOcclusion=true
  lPFFrequencyAtMin=10000
  lPFRadiusMin=2000
  nonFocusVolumeAttenuation=0.7
  occlusionLowPassFilterFrequency=4000
  occlusionTraceChannel=ECC_GameTraceChannel11
  occlusionVolumeAttenuation=4
 }
 concurrencyOverrides {bLimitToOwner=true}
}
