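# Sound cue for the small-scale grind rail boost intro, with separate local, enemy and other-player mixes.
# IsLocal routes to first-person attenuation, mixing a random, briefly delayed intro with a random whoosh.
# Otherwise only the delayed intro plays through third-person attenuation, at full level for IsEnemy and heavily reduced by a Modulator for others.
# A concurrency override allows one instance and blocks new ones while it plays.
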SoundCue /GrindRail/Audio/SmallScaleGrinding/Small_Grinding_Boost_IntroB
root=Branch_2
nodes [
 {id=Attenuation_1, children [Delay_0], properties {attenuationSettings=/GrindRail/Audio/Grinding_3P_Att}}
 {id=Attenuation_2, children [Mixer_0], properties {attenuationSettings=/GrindRail/Audio/Grinding_1P_Att}}
 {id=Branch_0, children [Attenuation_1, Modulator_4, Modulator_4], properties {boolParameterName=IsEnemy}}
 {id=Branch_2, children [Attenuation_2, Branch_0, Branch_0], properties {boolParameterName=IsLocal}}
 {id=Delay_0, children [Random_0], properties {delayMax=0.1, delayMin=0.1}}
 {id=Mixer_0, children [Delay_0,Random_1], properties {inputVolume [1,1]}}
 {id=Modulator_4, children [Attenuation_1], properties {volumeMax=0.3, volumeMin=0.2}}
 {id=Random_0, children [WavePlayer_0,WavePlayer_1,WavePlayer_2], properties {bShouldExcludeFromBranchCulling=true, weights [1,1,1]}}
 {
  id=Random_1
  children [WavePlayer_3,WavePlayer_4,WavePlayer_5,WavePlayer_6,WavePlayer_7]
  properties {bShouldExcludeFromBranchCulling=true, weights [1,1,1,1,1]}
 }
 {id=WavePlayer_0, properties {soundWaveAssetPtr=&/Small_Grinding_Boost_Intro_03}}
 {id=WavePlayer_1, properties {soundWaveAssetPtr=&/Small_Grinding_Boost_Intro_02}}
 {id=WavePlayer_2, properties {soundWaveAssetPtr=&/Small_Grinding_Boost_Intro_01}}
 {id=WavePlayer_3, properties {soundWaveAssetPtr=&/Small_Grinding_Boost_Whoosh_01}}
 {id=WavePlayer_4, properties {soundWaveAssetPtr=&/Small_Grinding_Boost_Whoosh_02}}
 {id=WavePlayer_5, properties {soundWaveAssetPtr=&/Small_Grinding_Boost_Whoosh_03}}
 {id=WavePlayer_6, properties {soundWaveAssetPtr=&/Small_Grinding_Boost_Whoosh_04}}
 {id=WavePlayer_7, properties {soundWaveAssetPtr=&/Small_Grinding_Boost_Whoosh_05}}
]
properties {
 bOverrideConcurrency=true
 maxDistance=6500
 soundClassObject=&Fort_Player_Foley_Athena
 volumeMultiplier=0.7
 concurrencyOverrides {maxCount=1, resolutionRule=PreventNew}
}
