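# Sound cue for landing onto a small-scale grind rail, classed as player foley and limited to one instance per owner.
# IsLocal selects a first-person mix of randomized jump-on, stinger, sparks and slightly delayed land layers.
# Otherwise, a third-person mix drops the stinger and delay and uses its own attenuation.
# IsEnemy plays that mix at full level, while non-enemy or unset routes it through an extra modulator that makes it much quieter.
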SoundCue /GrindRail/Audio/SmallScaleGrinding/Small_Grinding_Enter_Land
root=Branch_2
nodes [
 {id=Attenuation_6, children [Modulator_0], properties {attenuationSettings=/GrindRail/Audio/Grinding_1P_Att}}
 {id=Attenuation_7, children [Modulator_1], properties {attenuationSettings=/GrindRail/Audio/Grinding_3P_Att}}
 {id=Branch_0, children [Attenuation_7, Modulator_4, Modulator_4], properties {boolParameterName=IsEnemy}}
 {id=Branch_2, children [Attenuation_6, Branch_0, Branch_0], properties {boolParameterName=IsLocal}}
 {id=Delay_0, children [Random_9], properties {delayMax=0.08, delayMin=0.08}}
 {id=Mixer_0, children [Random_10,Random_4,Random_8], properties {inputVolume [1,1,0.9]}}
 {id=Mixer_1, children [Delay_0,Random_0,Random_1,Random_7], properties {inputVolume [1,0.4,0.4,0.9]}}
 {id=Modulator_0, children [Mixer_1], properties {pitchMin=0.98, volumeMax=0.9, volumeMin=0.85}}
 {id=Modulator_1, children [Mixer_0], properties {pitchMin=0.98, volumeMax=2, volumeMin=2}}
 {id=Modulator_4, children [Attenuation_7], properties {volumeMax=0.3, volumeMin=0.2}}
 {
  id=Random_0
  children [WavePlayer_0, WavePlayer_1, WavePlayer_4, WavePlayer_5, WavePlayer_14]
  properties {bShouldExcludeFromBranchCulling=true, weights [1,1,1,1,1]}
 }
 {
  id=Random_1
  children [WavePlayer_15, WavePlayer_16, WavePlayer_17, WavePlayer_18, WavePlayer_19]
  properties {bShouldExcludeFromBranchCulling=true, weights [1,1,1,1,1]}
 }
 {
  id=Random_10
  children [WavePlayer_50, WavePlayer_51, WavePlayer_52, WavePlayer_53, WavePlayer_54]
  properties {bShouldExcludeFromBranchCulling=true, weights [1,1,1,1,1]}
 }
 {
  id=Random_4
  children [WavePlayer_20, WavePlayer_21, WavePlayer_22, WavePlayer_23, WavePlayer_24]
  properties {bShouldExcludeFromBranchCulling=true, weights [1,1,1,1,1]}
 }
 {
  id=Random_7
  children [WavePlayer_35, WavePlayer_36, WavePlayer_37, WavePlayer_38, WavePlayer_39]
  properties {bShouldExcludeFromBranchCulling=true, weights [1,1,1,1,1]}
 }
 {
  id=Random_8
  children [WavePlayer_40, WavePlayer_41, WavePlayer_42, WavePlayer_43, WavePlayer_44]
  properties {bShouldExcludeFromBranchCulling=true, weights [1,1,1,1,1]}
 }
 {
  id=Random_9
  children [WavePlayer_45, WavePlayer_46, WavePlayer_47, WavePlayer_48, WavePlayer_49]
  properties {bShouldExcludeFromBranchCulling=true, weights [1,1,1,1,1]}
 }
 {id=WavePlayer_0, properties {soundWaveAssetPtr=&/Small_Grinding_JumpOn_Stinger_05}}
 {id=WavePlayer_1, properties {soundWaveAssetPtr=&/Small_Grinding_JumpOn_Stinger_04}}
 {id=WavePlayer_14, properties {soundWaveAssetPtr=&/Small_Grinding_JumpOn_Stinger_01}}
 {id=WavePlayer_15, properties {soundWaveAssetPtr=&/Small_Grinding_JumpLand_Sparks_05}}
 {id=WavePlayer_16, properties {soundWaveAssetPtr=&/Small_Grinding_JumpLand_Sparks_04}}
 {id=WavePlayer_17, properties {soundWaveAssetPtr=&/Small_Grinding_JumpLand_Sparks_03}}
 {id=WavePlayer_18, properties {soundWaveAssetPtr=&/Small_Grinding_JumpLand_Sparks_02}}
 {id=WavePlayer_19, properties {soundWaveAssetPtr=&/Small_Grinding_JumpLand_Sparks_01}}
 {id=WavePlayer_20, properties {soundWaveAssetPtr=&Small_Grinding_JumpLand_Sparks_05}}
 {id=WavePlayer_21, properties {soundWaveAssetPtr=&Small_Grinding_JumpLand_Sparks_04}}
 {id=WavePlayer_22, properties {soundWaveAssetPtr=&Small_Grinding_JumpLand_Sparks_03}}
 {id=WavePlayer_23, properties {soundWaveAssetPtr=&Small_Grinding_JumpLand_Sparks_02}}
 {id=WavePlayer_24, properties {soundWaveAssetPtr=&Small_Grinding_JumpLand_Sparks_01}}
 {id=WavePlayer_35, properties {soundWaveAssetPtr=&/Small_Grinding_JumpOn_01}}
 {id=WavePlayer_36, properties {soundWaveAssetPtr=&/Small_Grinding_JumpOn_02}}
 {id=WavePlayer_37, properties {soundWaveAssetPtr=&/Small_Grinding_JumpOn_03}}
 {id=WavePlayer_38, properties {soundWaveAssetPtr=&/Small_Grinding_JumpOn_04}}
 {id=WavePlayer_39, properties {soundWaveAssetPtr=&/Small_Grinding_JumpOn_05}}
 {id=WavePlayer_4, properties {soundWaveAssetPtr=&/Small_Grinding_JumpOn_Stinger_03}}
 {id=WavePlayer_40, properties {soundWaveAssetPtr=&Small_Grinding_JumpOn_01}}
 {id=WavePlayer_41, properties {soundWaveAssetPtr=&Small_Grinding_JumpOn_02}}
 {id=WavePlayer_42, properties {soundWaveAssetPtr=&Small_Grinding_JumpOn_03}}
 {id=WavePlayer_43, properties {soundWaveAssetPtr=&Small_Grinding_JumpOn_04}}
 {id=WavePlayer_44, properties {soundWaveAssetPtr=&Small_Grinding_JumpOn_05}}
 {id=WavePlayer_45, properties {soundWaveAssetPtr=&/Small_Grinding_JumpLand_01}}
 {id=WavePlayer_46, properties {soundWaveAssetPtr=&/Small_Grinding_JumpLand_02}}
 {id=WavePlayer_47, properties {soundWaveAssetPtr=&/Small_Grinding_JumpLand_03}}
 {id=WavePlayer_48, properties {soundWaveAssetPtr=&/Small_Grinding_JumpLand_04}}
 {id=WavePlayer_49, properties {soundWaveAssetPtr=&/Small_Grinding_JumpLand_05}}
 {id=WavePlayer_5, properties {soundWaveAssetPtr=&/Small_Grinding_JumpOn_Stinger_02}}
 {id=WavePlayer_50, properties {soundWaveAssetPtr=&Small_Grinding_JumpLand_01}}
 {id=WavePlayer_51, properties {soundWaveAssetPtr=&Small_Grinding_JumpLand_02}}
 {id=WavePlayer_52, properties {soundWaveAssetPtr=&Small_Grinding_JumpLand_03}}
 {id=WavePlayer_53, properties {soundWaveAssetPtr=&Small_Grinding_JumpLand_04}}
 {id=WavePlayer_54, properties {soundWaveAssetPtr=&Small_Grinding_JumpLand_05}}
]
properties {
 bOverrideConcurrency=true
 maxDistance=6500
 soundClassObject=&Fort_Player_Foley_Athena
 volumeMultiplier=1.2
 concurrencyOverrides {bLimitToOwner=true, maxCount=1, resolutionRule=PreventNew}
}
