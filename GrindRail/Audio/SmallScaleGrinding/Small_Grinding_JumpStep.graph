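# Sound cue for the jump step on small-scale rail grinding, branching on IsLocal and IsEnemy.
# Local playback mixes a random step-jump sample with a sweetener through first-person attenuation.
# Remote playback uses third-person attenuation, adding random jump-on start samples and a higher-pitched sweetener.
# Enemies play at full mix level, while non-enemy or unset cases pass through a modulator that sharply lowers volume.
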SoundCue /GrindRail/Audio/SmallScaleGrinding/Small_Grinding_JumpStep
root=Branch_2
nodes [
 {id=Attenuation_0, children [Modulator_2], properties {attenuationSettings=/GrindRail/Audio/Grinding_1P_Att}}
 {id=Attenuation_6, children [Modulator_0], properties {attenuationSettings=&Grinding_1P_Att}}
 {id=Attenuation_7, children [Mixer_1], properties {attenuationSettings=/GrindRail/Audio/Grinding_3P_Att}}
 {id=Branch_0, children [Attenuation_7, Modulator_4, Modulator_4], properties {boolParameterName=IsEnemy}}
 {id=Branch_2, children [Mixer_0,Branch_0,Branch_0], properties {boolParameterName=IsLocal}}
 {id=Mixer_0, children [Attenuation_6, Attenuation_0], properties {inputVolume [1,1.2]}}
 {id=Mixer_1, children [Modulator_1,Modulator_3,Modulator_5], properties {inputVolume [1,0.6,1.2]}}
 {id=Modulator_0, children [Random_0], properties {pitchMax=1.02, pitchMin=0.98}}
 {id=Modulator_1, children [Random_6], properties {pitchMax=1.02, pitchMin=0.98}}
 {id=Modulator_2, children [WavePlayer_10], properties {pitchMax=1.02, pitchMin=0.98}}
 {id=Modulator_3, children [Random_1], properties {pitchMax=1.02, pitchMin=0.98}}
 {id=Modulator_4, children [Attenuation_7], properties {volumeMax=0.3, volumeMin=0.2}}
 {id=Modulator_5, children [WavePlayer_11], properties {pitchMax=1.2, pitchMin=1}}
 {
  id=Random_0
  children [WavePlayer_0,WavePlayer_1,WavePlayer_2,WavePlayer_3,WavePlayer_4]
  properties {bShouldExcludeFromBranchCulling=true, weights [1,1,1,1,1]}
 }
 {
  id=Random_1
  children [WavePlayer_5,WavePlayer_6,WavePlayer_7,WavePlayer_8,WavePlayer_9]
  properties {bShouldExcludeFromBranchCulling=true, weights [1,1,1,1,1]}
 }
 {
  id=Random_6
  children [WavePlayer_30, WavePlayer_31, WavePlayer_32, WavePlayer_33, WavePlayer_34]
  properties {bShouldExcludeFromBranchCulling=true, weights [1,1,1,1,1]}
 }
 {id=WavePlayer_0, properties {soundWaveAssetPtr=&/Small_Grinding_Step_Jump_01}}
 {id=WavePlayer_1, properties {soundWaveAssetPtr=&/Small_Grinding_Step_Jump_02}}
 {id=WavePlayer_10, properties {soundWaveAssetPtr=&/GrindRail_Jump_Sweetner}}
 {id=WavePlayer_11, properties {soundWaveAssetPtr=&GrindRail_Jump_Sweetner}}
 {id=WavePlayer_2, properties {soundWaveAssetPtr=&/Small_Grinding_Step_Jump_03}}
 {id=WavePlayer_3, properties {soundWaveAssetPtr=&/Small_Grinding_Step_Jump_04}}
 {id=WavePlayer_30, properties {soundWaveAssetPtr=&Small_Grinding_Step_Jump_01}}
 {id=WavePlayer_31, properties {soundWaveAssetPtr=&Small_Grinding_Step_Jump_02}}
 {id=WavePlayer_32, properties {soundWaveAssetPtr=&Small_Grinding_Step_Jump_03}}
 {id=WavePlayer_33, properties {soundWaveAssetPtr=&Small_Grinding_Step_Jump_04}}
 {id=WavePlayer_34, properties {soundWaveAssetPtr=&/Small_Grinding_Step_Jump_05}}
 {id=WavePlayer_4, properties {soundWaveAssetPtr=&Small_Grinding_Step_Jump_05}}
 {id=WavePlayer_5, properties {soundWaveAssetPtr=&/Small_Grinding_JumpOn_Start_01}}
 {id=WavePlayer_6, properties {soundWaveAssetPtr=&/Small_Grinding_JumpOn_Start_02}}
 {id=WavePlayer_7, properties {soundWaveAssetPtr=&/Small_Grinding_JumpOn_Start_04}}
 {id=WavePlayer_8, properties {soundWaveAssetPtr=&/Small_Grinding_JumpOn_Start_05}}
 {id=WavePlayer_9, properties {soundWaveAssetPtr=&/Small_Grinding_JumpOn_Start_03}}
]
properties {maxDistance=6500, soundClassObject=&Fort_Gadgets, volumeMultiplier=0.6}
