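# Sound cue for the jump-up sound in small-scale rail grinding, classed as a Fort gadget sound.
# An IsLocal branch selects a first-person attenuated mix; otherwise an IsEnemy branch plays the third-person mix at full level for enemies and heavily reduced for others.
# Both mixes layer randomized jump-up, sparks and jump-on start waves, with slight pitch variation.
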SoundCue /GrindRail/Audio/SmallScaleGrinding/Small_Grinding_JumpUp
root=Branch_3
nodes [
 {id=Attenuation_6, children [Modulator_0], properties {attenuationSettings=/GrindRail/Audio/Grinding_1P_Att}}
 {id=Attenuation_7, children [Mixer_1], properties {attenuationSettings=/GrindRail/Audio/Grinding_3P_Att}}
 {id=Branch_0, children [Attenuation_7, Modulator_4, Modulator_4], properties {boolParameterName=IsEnemy}}
 {id=Branch_3, children [Attenuation_6, Branch_0, Branch_0], properties {boolParameterName=IsLocal}}
 {id=Mixer_1, children [Modulator_1,Modulator_3,Random_3], properties {inputVolume [1.5,1,1.5]}}
 {id=Mixer_2, children [Random_4,Random_2,Modulator_2], properties {inputVolume [1,0.7,1]}}
 {id=Modulator_0, children [Mixer_2], properties {pitchMax=1.02, pitchMin=0.98}}
 {id=Modulator_1, children [Random_5], properties {pitchMax=1.02, pitchMin=0.98}}
 {id=Modulator_2, children [Random_0], properties {pitchMax=1.02, pitchMin=0.98}}
 {id=Modulator_3, children [Random_1], properties {pitchMax=1.02, pitchMin=0.98}}
 {id=Modulator_4, children [Attenuation_7], properties {volumeMax=0.15, volumeMin=0.1}}
 {id=Random_0, children [WavePlayer_0,WavePlayer_1,WavePlayer_2,WavePlayer_3,WavePlayer_4], properties {weights [1,1,1,1,1]}}
 {id=Random_1, children [WavePlayer_5,WavePlayer_6,WavePlayer_7,WavePlayer_8,WavePlayer_9], properties {weights [1,1,1,1,1]}}
 {
  id=Random_2
  children [WavePlayer_10, WavePlayer_11, WavePlayer_12, WavePlayer_13, WavePlayer_14]
  properties {bShouldExcludeFromBranchCulling=true, weights [1,1,1,1,1]}
 }
 {
  id=Random_3
  children [WavePlayer_15, WavePlayer_26, WavePlayer_27, WavePlayer_28, WavePlayer_29]
  properties {bShouldExcludeFromBranchCulling=true, weights [1,1,1,1,1]}
 }
 {
  id=Random_4
  children [WavePlayer_16, WavePlayer_17, WavePlayer_18, WavePlayer_19, WavePlayer_20]
  properties {bShouldExcludeFromBranchCulling=true, weights [1,1,1,1,1]}
 }
 {id=Random_5, children [WavePlayer_21, WavePlayer_22, WavePlayer_23, WavePlayer_24, WavePlayer_25], properties {weights [1,1,1,1,1]}}
 {id=WavePlayer_0, properties {soundWaveAssetPtr=&/Small_Grinding_JumpOn_Start_01}}
 {id=WavePlayer_1, properties {soundWaveAssetPtr=&/Small_Grinding_JumpOn_Start_02}}
 {id=WavePlayer_10, properties {soundWaveAssetPtr=&/Small_Grinding_JumpUp_Sparks_05}}
 {id=WavePlayer_11, properties {soundWaveAssetPtr=&/Small_Grinding_JumpUp_Sparks_04}}
 {id=WavePlayer_12, properties {soundWaveAssetPtr=&/Small_Grinding_JumpUp_Sparks_03}}
 {id=WavePlayer_13, properties {soundWaveAssetPtr=&/Small_Grinding_JumpUp_Sparks_02}}
 {id=WavePlayer_14, properties {soundWaveAssetPtr=&/Small_Grinding_JumpUp_Sparks_01}}
 {id=WavePlayer_15, properties {soundWaveAssetPtr=&Small_Grinding_JumpUp_Sparks_05}}
 {id=WavePlayer_16, properties {soundWaveAssetPtr=&/Small_Grinding_JumpUp_01}}
 {id=WavePlayer_17, properties {soundWaveAssetPtr=&/Small_Grinding_JumpUp_02}}
 {id=WavePlayer_18, properties {soundWaveAssetPtr=&/Small_Grinding_JumpUp_03}}
 {id=WavePlayer_19, properties {soundWaveAssetPtr=&/Small_Grinding_JumpUp_04}}
 {id=WavePlayer_2, properties {soundWaveAssetPtr=&/Small_Grinding_JumpOn_Start_04}}
 {id=WavePlayer_20, properties {soundWaveAssetPtr=&/Small_Grinding_JumpUp_05}}
 {id=WavePlayer_21, properties {soundWaveAssetPtr=&Small_Grinding_JumpUp_01}}
 {id=WavePlayer_22, properties {soundWaveAssetPtr=&Small_Grinding_JumpUp_02}}
 {id=WavePlayer_23, properties {soundWaveAssetPtr=&Small_Grinding_JumpUp_03}}
 {id=WavePlayer_24, properties {soundWaveAssetPtr=&Small_Grinding_JumpUp_04}}
 {id=WavePlayer_25, properties {soundWaveAssetPtr=&Small_Grinding_JumpUp_05}}
 {id=WavePlayer_26, properties {soundWaveAssetPtr=&Small_Grinding_JumpUp_Sparks_04}}
 {id=WavePlayer_27, properties {soundWaveAssetPtr=&Small_Grinding_JumpUp_Sparks_03}}
 {id=WavePlayer_28, properties {soundWaveAssetPtr=&Small_Grinding_JumpUp_Sparks_02}}
 {id=WavePlayer_29, properties {soundWaveAssetPtr=&Small_Grinding_JumpUp_Sparks_01}}
 {id=WavePlayer_3, properties {soundWaveAssetPtr=&/Small_Grinding_JumpOn_Start_05}}
 {id=WavePlayer_4, properties {soundWaveAssetPtr=&/Small_Grinding_JumpOn_Start_03}}
 {id=WavePlayer_5, properties {soundWaveAssetPtr=&Small_Grinding_JumpOn_Start_01}}
 {id=WavePlayer_6, properties {soundWaveAssetPtr=&Small_Grinding_JumpOn_Start_02}}
 {id=WavePlayer_7, properties {soundWaveAssetPtr=&Small_Grinding_JumpOn_Start_04}}
 {id=WavePlayer_8, properties {soundWaveAssetPtr=&Small_Grinding_JumpOn_Start_05}}
 {id=WavePlayer_9, properties {soundWaveAssetPtr=&Small_Grinding_JumpOn_Start_03}}
]
properties {maxDistance=6500, soundClassObject=&Fort_Gadgets, volumeMultiplier=1.2}
