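# Sound cue for the jump-up one-shot in small-scale rail grinding.
# It mixes randomized jump-up, sparks and jump-on start variations with a pitch-modulated sweetener, then applies slight overall pitch variation.
# IsLocal selects first-person attenuation; otherwise the sound uses third-person attenuation, at full level when IsEnemy is true and heavily reduced in volume when false or unset.
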SoundCue /GrindRail/Audio/SmallScaleGrinding/Small_Grinding_JumpUpB
root=Branch_2
nodes [
 {id=Attenuation_6, children [Modulator_0], properties {attenuationSettings=/GrindRail/Audio/Grinding_1P_Att}}
 {id=Attenuation_7, children [Modulator_0], properties {attenuationSettings=/GrindRail/Audio/Grinding_3P_Att}}
 {id=Branch_0, children [Attenuation_7, Modulator_4, Modulator_4], properties {boolParameterName=IsEnemy}}
 {id=Branch_2, children [Attenuation_6, Branch_0, Branch_0], properties {boolParameterName=IsLocal}}
 {id=Mixer_2, children [Random_4,Random_2,Modulator_2,Modulator_1], properties {inputVolume [1.3,1,1,1]}}
 {id=Modulator_0, children [Mixer_2], properties {pitchMax=1.02, pitchMin=0.98}}
 {id=Modulator_1, children [WavePlayer_5], properties {pitchMax=1.2, pitchMin=0.9}}
 {id=Modulator_2, children [Random_0], properties {pitchMax=1.02, pitchMin=0.98}}
 {id=Modulator_4, children [Attenuation_7], properties {volumeMax=0.3, volumeMin=0.2}}
 {
  id=Random_0
  children [WavePlayer_0,WavePlayer_1,WavePlayer_2,WavePlayer_3,WavePlayer_4]
  properties {bShouldExcludeFromBranchCulling=true, weights [1,1,1,1,1]}
 }
 {
  id=Random_2
  children [WavePlayer_10, WavePlayer_11, WavePlayer_12, WavePlayer_13, WavePlayer_14]
  properties {bShouldExcludeFromBranchCulling=true, weights [1,1,1,1,1]}
 }
 {
  id=Random_4
  children [WavePlayer_16, WavePlayer_17, WavePlayer_18, WavePlayer_19, WavePlayer_20]
  properties {bShouldExcludeFromBranchCulling=true, weights [1,1,1,1,1]}
 }
 {id=WavePlayer_0, properties {soundWaveAssetPtr=&/Small_Grinding_JumpOn_Start_01}}
 {id=WavePlayer_1, properties {soundWaveAssetPtr=&/Small_Grinding_JumpOn_Start_02}}
 {id=WavePlayer_10, properties {soundWaveAssetPtr=&/Small_Grinding_JumpUp_Sparks_05}}
 {id=WavePlayer_11, properties {soundWaveAssetPtr=&/Small_Grinding_JumpUp_Sparks_04}}
 {id=WavePlayer_12, properties {soundWaveAssetPtr=&/Small_Grinding_JumpUp_Sparks_03}}
 {id=WavePlayer_13, properties {soundWaveAssetPtr=&/Small_Grinding_JumpUp_Sparks_02}}
 {id=WavePlayer_14, properties {soundWaveAssetPtr=&/Small_Grinding_JumpUp_Sparks_01}}
 {id=WavePlayer_16, properties {soundWaveAssetPtr=&/Small_Grinding_JumpUp_01}}
 {id=WavePlayer_17, properties {soundWaveAssetPtr=&/Small_Grinding_JumpUp_02}}
 {id=WavePlayer_18, properties {soundWaveAssetPtr=&/Small_Grinding_JumpUp_03}}
 {id=WavePlayer_19, properties {soundWaveAssetPtr=&/Small_Grinding_JumpUp_04}}
 {id=WavePlayer_2, properties {soundWaveAssetPtr=&/Small_Grinding_JumpOn_Start_04}}
 {id=WavePlayer_20, properties {soundWaveAssetPtr=&/Small_Grinding_JumpUp_05}}
 {id=WavePlayer_3, properties {soundWaveAssetPtr=&/Small_Grinding_JumpOn_Start_05}}
 {id=WavePlayer_4, properties {soundWaveAssetPtr=&/Small_Grinding_JumpOn_Start_03}}
 {id=WavePlayer_5, properties {soundWaveAssetPtr=&/GrindRail_Jump_Sweetner}}
]
properties {maxDistance=6500, soundClassObject=&Fort_Gadgets, volumeMultiplier=0.8}
