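# Sound cue for the lean-forward sound of small-scale rail grinding, classed as player foley.
# It randomly picks one of five reused jump-land spark waves and lowers its pitch slightly.
# An IsLocal branch selects first-person attenuation, otherwise an IsEnemy branch selects third-person attenuation, at much reduced volume unless the source is an enemy.
# All paths share the random node, which is excluded from branch culling.
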SoundCue /GrindRail/Audio/SmallScaleGrinding/Small_Grinding_LeanForward
root=Branch_2
nodes [
 {id=Attenuation_6, children [Modulator_0], properties {attenuationSettings=/GrindRail/Audio/Grinding_1P_Att}}
 {id=Attenuation_7, children [Modulator_0], properties {attenuationSettings=/GrindRail/Audio/Grinding_3P_Att}}
 {id=Branch_0, children [Attenuation_7, Modulator_4, Modulator_4], properties {boolParameterName=IsEnemy}}
 {id=Branch_2, children [Attenuation_6, Branch_0, Branch_0], properties {boolParameterName=IsLocal}}
 {id=Modulator_0, children [Random_0], properties {pitchMax=0.9, pitchMin=0.8}}
 {id=Modulator_4, children [Attenuation_7], properties {volumeMax=0.3, volumeMin=0.2}}
 {
  id=Random_0
  children [WavePlayer_0,WavePlayer_1,WavePlayer_2,WavePlayer_3,WavePlayer_4]
  properties {bShouldExcludeFromBranchCulling=true, weights [1,1,1,1,1]}
 }
 {id=WavePlayer_0, properties {soundWaveAssetPtr=&/Small_Grinding_JumpLand_Sparks_05}}
 {id=WavePlayer_1, properties {soundWaveAssetPtr=&/Small_Grinding_JumpLand_Sparks_04}}
 {id=WavePlayer_2, properties {soundWaveAssetPtr=&/Small_Grinding_JumpLand_Sparks_03}}
 {id=WavePlayer_3, properties {soundWaveAssetPtr=&/Small_Grinding_JumpLand_Sparks_02}}
 {id=WavePlayer_4, properties {soundWaveAssetPtr=&/Small_Grinding_JumpLand_Sparks_01}}
]
properties {maxDistance=6500, soundClassObject=&Fort_Player_Foley_Athena, volumeMultiplier=0.95}
