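# Looping Sound Cue for small-scale rail grinding, with separate first-person and third-person mixes.
# IsLocal selects the 1P mix with an added wind-buffet layer, otherwise the 3P mix, which IsEnemy keeps at full level while non-enemies are heavily attenuated.
# Both mixes crossfade in randomly chosen base and boost loops, with GrindSpeed driving volume and pitch and IsGrinding scaling overall volume.
# A root Enveloper shapes the fade-in, and overridden concurrency caps simultaneous voices.
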SoundCue /GrindRail/Audio/SmallScaleGrinding/Small_Grinding_Loop
root=Enveloper_0
nodes [
 {id=Attenuation_0, children [ModulatorContinuous_10], properties {attenuationSettings=/GrindRail/Audio/Grinding_1P_Att}}
 {id=Attenuation_1, children [Modulator_0], properties {attenuationSettings=/GrindRail/Audio/Grinding_3P_Loops_Att}}
 {id=Branch_1, children [Attenuation_1, Modulator_4, Modulator_4], properties {boolParameterName=IsEnemy}}
 {id=Branch_2, children [Attenuation_0, Branch_1, Branch_1], properties {boolParameterName=IsLocal}}
 {
  id=Enveloper_0
  children [Branch_2]
  properties {pitchCurve {editorCurveData {keys [{time=0}]}}, volumeCurve {editorCurveData {keys [<at 0 0>,<at 0.0495 0.9602>,<at 1 1>]}}}
 }
 {id=Mixer_0, children [ModulatorContinuous_3, ModulatorContinuous_4], properties {inputVolume [1.2,1.2]}}
 {id=Mixer_1, children [ModulatorContinuous_2, ModulatorContinuous_5], properties {inputVolume [1,1.5]}}
 {id=Mixer_3, children [Mixer_0, ModulatorContinuous_12], properties {inputVolume [1,1.5]}}
 {id=ModulatorContinuous_10, children [Mixer_3], properties {volumeModulationParams {default=0, parameterName=IsGrinding}}}
 {
  id=ModulatorContinuous_12
  children [Random_4]
  properties {
   pitchModulationParams {maxOutput=1.1, minOutput=0.9, parameterName=GrindSpeed}
   volumeModulationParams {default=0, minInput=0.4, parameterName=GrindSpeed}
  }
 }
 {
  id=ModulatorContinuous_2
  children [Random_1]
  properties {
   pitchModulationParams {minOutput=0.95, parameterName=GrindSpeed}
   volumeModulationParams {default=0, minOutput=0.2, parameterName=GrindSpeed}
  }
 }
 {
  id=ModulatorContinuous_3
  children [Random_0]
  properties {
   pitchModulationParams {minOutput=0.95, parameterName=GrindSpeed}
   volumeModulationParams {default=0, minOutput=0.2, parameterName=GrindSpeed}
  }
 }
 {
  id=ModulatorContinuous_4
  children [WavePlayer_5]
  properties {
   pitchModulationParams {maxOutput=1.1, minOutput=0.8, parameterName=GrindSpeed}
   volumeModulationParams {default=0, maxOutput=0.6, minInput=0.35, minOutput=0.1, parameterName=GrindSpeed}
  }
 }
 {id=ModulatorContinuous_47, children [Mixer_1], properties {volumeModulationParams {default=0, parameterName=IsGrinding}}}
 {
  id=ModulatorContinuous_5
  children [Random_2]
  properties {
   pitchModulationParams {maxOutput=1.1, minOutput=0.9, parameterName=GrindSpeed}
   volumeModulationParams {default=0, minInput=0.4, parameterName=GrindSpeed}
  }
 }
 {id=Modulator_0, children [ModulatorContinuous_47], properties {pitchMax=1, pitchMin=1, volumeMax=1.5, volumeMin=1.5}}
 {id=Modulator_4, children [Attenuation_1], properties {volumeMax=0.3, volumeMin=0.2}}
 {id=Random_0, children [WavePlayer_16, WavePlayer_0], properties {bShouldExcludeFromBranchCulling=true, weights [1,1]}}
 {id=Random_1, children [WavePlayer_1,WavePlayer_2], properties {bShouldExcludeFromBranchCulling=true, weights [1,1]}}
 {id=Random_2, children [WavePlayer_4,WavePlayer_6], properties {bShouldExcludeFromBranchCulling=true, weights [1,1]}}
 {id=Random_4, children [WavePlayer_13, WavePlayer_14], properties {bShouldExcludeFromBranchCulling=true, weights [1,1]}}
 {id=WavePlayer_0, properties {bLooping=true, soundWaveAssetPtr=&/Small_Grinding_Loop_02}}
 {id=WavePlayer_1, properties {bLooping=true, soundWaveAssetPtr=&/Small_Grinding_Loop_01}}
 {id=WavePlayer_13, properties {bLooping=true, soundWaveAssetPtr=&/Small_Grinding_Boost_Loop_01}}
 {id=WavePlayer_14, properties {bLooping=true, soundWaveAssetPtr=&/Small_Grinding_Boost_Loop_02}}
 {id=WavePlayer_16, properties {bLooping=true, soundWaveAssetPtr=&Small_Grinding_Loop_01}}
 {id=WavePlayer_2, properties {bLooping=true, soundWaveAssetPtr=&Small_Grinding_Loop_02}}
 {id=WavePlayer_4, properties {bLooping=true, soundWaveAssetPtr=&Small_Grinding_Boost_Loop_01}}
 {id=WavePlayer_5, properties {bLooping=true, soundWaveAssetPtr=/GrindRail/Audio/Grind_Rail_Speed_Loop_Wind_Buffet}}
 {id=WavePlayer_6, properties {bLooping=true, soundWaveAssetPtr=&Small_Grinding_Boost_Loop_02}}
]
properties {
 bOverrideConcurrency=true
 maxDistance=8500
 soundClassObject=&Fort_Gadgets
 volumeMultiplier=1.4
 attenuationOverrides {
  bEnableOcclusion=true
  lPFFrequencyAtMin=10000
  lPFRadiusMin=2000
  nonFocusVolumeAttenuation=0.7
  occlusionLowPassFilterFrequency=4000
  occlusionTraceChannel=ECC_GameTraceChannel11
  occlusionVolumeAttenuation=4
 }
 concurrencyOverrides {bVolumeScaleCanRelease=true, maxCount=4, voiceStealReleaseTime=1}
}
