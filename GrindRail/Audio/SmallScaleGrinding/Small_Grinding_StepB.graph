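# Sound cue for a small-scale grind rail step sound (StepB), played in the Fort_Gadgets sound class.
# A random pick from six step waves gets slight pitch variation.
# An IsLocal branch sends local playback through first-person attenuation and everything else to an IsEnemy branch.
# Both IsEnemy paths use third-person attenuation, with non-enemy or unset cases cut to roughly a quarter volume.
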SoundCue /GrindRail/Audio/SmallScaleGrinding/Small_Grinding_StepB
root=Branch_2
nodes [
 {id=Attenuation_2, children [Modulator_1], properties {attenuationSettings=/GrindRail/Audio/Grinding_3P_Att}}
 {id=Attenuation_3, children [Modulator_0], properties {attenuationSettings=/GrindRail/Audio/Grinding_1P_Att}}
 {id=Branch_1, children [Attenuation_2, Modulator_4, Modulator_4], properties {boolParameterName=IsEnemy}}
 {id=Branch_2, children [Attenuation_3, Branch_1, Branch_1], properties {boolParameterName=IsLocal}}
 {id=Modulator_0, children [Random_0], properties {pitchMax=1.02, pitchMin=0.98}}
 {id=Modulator_1, children [Random_0], properties {pitchMax=1.02, pitchMin=0.98}}
 {id=Modulator_4, children [Attenuation_2], properties {volumeMax=0.3, volumeMin=0.2}}
 {
  id=Random_0
  children [WavePlayer_0,WavePlayer_1,WavePlayer_2,WavePlayer_3,WavePlayer_4,WavePlayer_5]
  properties {bShouldExcludeFromBranchCulling=true, weights [1,1,1,1,1,1]}
 }
 {id=WavePlayer_0, properties {soundWaveAssetPtr=&/Small_Grinding_Step_01}}
 {id=WavePlayer_1, properties {soundWaveAssetPtr=&/Small_Grinding_Step_02}}
 {id=WavePlayer_2, properties {soundWaveAssetPtr=&/Small_Grinding_Step_03}}
 {id=WavePlayer_3, properties {soundWaveAssetPtr=&/Small_Grinding_Step_04}}
 {id=WavePlayer_4, properties {soundWaveAssetPtr=&/Small_Grinding_Step_05}}
 {id=WavePlayer_5, properties {soundWaveAssetPtr=&/Small_Grinding_Step_06}}
]
properties {
 attenuationSettings=&Grinding_3P_Att
 maxDistance=6500
 soundClassObject=&Fort_Gadgets
 volumeMultiplier=0.7
}
